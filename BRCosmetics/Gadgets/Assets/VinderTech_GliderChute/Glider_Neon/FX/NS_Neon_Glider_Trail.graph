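# Niagara effect for the Neon glider: color-cycling ribbon trails and point lights driven by the glider's skeletal mesh.
# The trail emitter builds four AnimTrail ribbons from bone/socket pairs, hue-shifts them at spawn and pushes particles along sampled rotation.
# Ribbons use eroding material and fade in with emitter age, with one trail group on a dimmer alpha curve.
# A second emitter keeps four persistent, hue-shifting lights attached to sampled sockets.
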NiagaraSystem /BRCosmetics/Gadgets/Assets/VinderTech_GliderChute/Glider_Neon/FX/NS_Neon_Glider_Trail
dependencies {usedDataInterfaces [&NiagaraDataInterfaceCurve], usedDynamicInputs [], usedModules [&SystemState]}
emitters {
 Empty {
  emitterUpdateScript [
   &EmitterState(
    Inactive Response=::NewEnumerator1
    Life Cycle Mode=::NewEnumerator1
    Loop Behavior=::NewEnumerator1
    Loop Duration Mode=::NewEnumerator1
    UseLoopDelay=false
   )
   &SpawnBurst_Instantaneous(Spawn Count=4, Use Spawn Probability=false)
  ]
  particleSpawnScript [
   &InitializeParticle(Sprite Size Mode=::NewEnumerator3, Uniform Sprite Size=50, UsePositionOffset=false, Write Material Random=true)
   &FN_SetSpawnGroup()
  ]
  particleUpdateScript [
   &ParticleState(Kill Particles When Lifetime Has Elapsed=false)
   /Game/Effects/Niagara/Modules/Color/FN_ColorHueShift(Color:LinearColor=<rgba 1 0 0 1>, Speed=0.25)
   &SkeletalMeshLocation(
    Bone / Socket Index=&SpawnGroup()
    Inherit Velocity=false
    MaskSpawnGroup=false
    Mesh Sampling Type=::NewEnumerator5
    Skeletal Mesh=$User.SkeletalMesh
    Socket Sampling Mode=::NewEnumerator5
   )
   &FN_Light_v3(Create Exponent Attribute=true, Create Radius Attribute=true, Exponent=2, Radius=150)
  ]
  rendererClasses [&NiagaraLightRendererProperties]
  renderers [{rendererClass=&NiagaraLightRendererProperties, rendererIndex=0}]
 }
 FN_AnimTrail_with_MasterMaterial {
  emitterUpdateScript [
   &EmitterState()
   &FN_SpawnAnimTrail_v2(Number of Anim Trails=4)
  ]
  particleSpawnScript [
   &FN_SetSpawnGroup()
   /SetVariables(Particles.SpawnGroupV2:NiagaraInt32=&ModuloInt(A=&SpawnGroup(), B=4)) {isSetParameters=true}
   /Game/Effects/Niagara/Modules/AnimTrail/FN_InitializeAnimTrail/FN_InitializeAnimTrailBoneIndecies(Second Bone or Socket Index=1)
   &FN_InitializeAnimTrailBoneIndecies(AnimTrail Selection=1, First Bone or Socket Index=2, Second Bone or Socket Index=3) {module=FN_InitializeAnimTrailBoneIndecies001}
   &FN_InitializeAnimTrailBoneIndecies(AnimTrail Selection=2, First Bone or Socket Index=4, Second Bone or Socket Index=5) {module=FN_InitializeAnimTrailBoneIndecies002}
   &FN_InitializeAnimTrailBoneIndecies(AnimTrail Selection=3, First Bone or Socket Index=6, Second Bone or Socket Index=7) {module=FN_InitializeAnimTrailBoneIndecies003}
   /Game/Effects/Niagara/Modules/AnimTrail/FN_InitializeAnimTrail/FN_InitializeAnimTrail_v3/FN_InitializeAnimTrail(AnimTrailWidth=0.95, Color:LinearColor=<rgba 0.8461 0 1 1>, Lifetime=0.28)
   &SkeletalMeshLocation(
    Apply To Particle Position=false
    Bone / Socket Index=&SpawnGroup()
    Inherit Velocity=false
    MaskSpawnGroup=false
    Mesh Sampling Type=::NewEnumerator5
    Skeletal Mesh=$User.SkeletalMesh
    Socket Sampling Mode=::NewEnumerator5
   )
   &FN_ColorHueShift(Color:LinearColor=<rgba 1 0 0 1>, Speed=0.25)
   &AddVelocity(Velocity=&RotateVectorByQuaternion(Vector To Rotate=<1000 0 0>, Quaternion=$Output.SkeletalMeshLocation.SampledRotation))
  ]
  particleUpdateScript [
   &SolveForcesAndVelocity(Clamp Velocity=false, Limit Acceleration=false)
   /Game/Effects/Niagara/Modules/AnimTrail/FN_UpdateAnimTrail/FN_UpdateAnimTrail_v3/FN_UpdateAnimTrail()
   &FN_RibbonMaterial(
    Color Scale U=0.2182
    Erosion Amount:NiagaraFloat=&FloatFromCurve(FloatCurve=DataInterface() {propertyValues {Curve {keys [<at 0 0.2>,<at 1 0.5>]}}})
    Erosion Pan Offset U:NiagaraFloat=$Particles.MaterialRandom
    Erosion Pan Speed U=0.3
    Erosion Power=2
    Erossion Scale U=0.1
   )
   &ScaleColor(
    Scale Alpha=&Multiply_Float(
     A=&FloatFromCurve(FloatCurve=DataInterface() {propertyValues {Curve {keys [<at 0 0>,<at 0.5 1>]}}}, CurveIndex=$Emitter.Age)
     B=&MakeCustomFloatFromBool(
      Bool=&SetBoolByIntComparison(A=$Particles.SpawnGroupV2, B=1, Comparison Type=::NewEnumerator0)
      True Float=&FloatFromCurve(FloatCurve=DataInterface() {propertyValues {Curve {keys [<at 0.25 0.3>,<at 1 0>]}}})
      False Float=&FloatFromCurve(FloatCurve=DataInterface() {propertyValues {Curve {keys [<at 0.25 0.5>,<at 1 0>]}}})
     )
    )
    ScaleA=true
    ScaleRGB=true
   )
  ]
  rendererClasses [&NiagaraRibbonRendererProperties]
  renderers [{rendererClass=&NiagaraRibbonRendererProperties}]
 }
}
summary {
 systemName=NS_Neon_Glider_Trail
 emitters [
  {emitterName=FN_AnimTrail_with_MasterMaterial, rendererClasses [&NiagaraRibbonRendererProperties]}
  {emitterName=Empty, rendererClasses [&NiagaraLightRendererProperties]}
 ]
 userVariables [
  User.RibbonIDOffset {type=NiagaraInt32}
  User.SkeletalMesh {
   type=NiagaraDataInterfaceSkeletalMesh
   defaultValue {
    struct=NiagaraExt_VariableValue_DataInterface
    value {dataInterface=&self:NiagaraDataInterfaceSkeletalMesh_0}
   }
  }
 ]
}
