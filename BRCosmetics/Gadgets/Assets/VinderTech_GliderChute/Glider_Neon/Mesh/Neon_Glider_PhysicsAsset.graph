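# Physics asset for the Neon glider mesh in the VinderTech GliderChute cosmetics, giving its frame, handles and wings physics bodies.
# Every bone body is a single capsule, with left and right sides mirrored.
# Constraints form a tree rooted at Frame: four wing chains, each forking into two tip segments, plus Handles_Base carrying both handles.
# All joints share the same moderate swing and twist limits.
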PhysicsAsset /BRCosmetics/Gadgets/Assets/VinderTech_GliderChute/Glider_Neon/Mesh/Neon_Glider_PhysicsAsset
bodies {
 Frame [{shapeType=Capsule, length=29.2101, radius=4.0496, center=<0 3.7196 -17.3351>, rotation=<rot 0 90 -90>}]
 Handle_l [{shapeType=Capsule, length=15.1622, radius=13.9597, center=<-5.8396 -0.1209 -5.3817>, rotation=<rot 0 -93.0375 -65.1567>}]
 Handle_r [{shapeType=Capsule, length=15.2069, radius=13.9239, center=<5.7661 0.1368 5.3511>, rotation=<rot 0 -92.9779 -64.7569>}]
 Handles_Base [{shapeType=Capsule, length=29.1044, radius=13.1533, center=<0.4281 7.5623 -51.812>, rotation=<rot 0 -90.9187 -89.7107>}]
 Wing_Lower_l [{shapeType=Capsule, length=37.9278, radius=28.659, center=<32.2134 -16.326 -0.018>, rotation=<rot 0 16.05 -89.9896>}]
 Wing_Lower_l1 [{shapeType=Capsule, length=38.7779, radius=27.6719, center=<18.2958 -8.5962 -0.0187>, rotation=<rot 0 171.5036 -89.9336>}]
 Wing_Lower_l2 [{shapeType=Capsule, length=24.0851, radius=14.0993, center=<19.3191 -4.3647 -0.018>, rotation=<rot 0 -7.3825 -89.9964>}]
 Wing_Lower_l3 [{shapeType=Capsule, length=38.4248, radius=6.0773, center=<36.9342 2.441 -0.0179>, rotation=<rot 0 -1.4308 -89.9377>}]
 Wing_Lower_r [{shapeType=Capsule, length=37.9278, radius=28.659, center=<-32.2135 16.3258 0.018>, rotation=<rot 0 16.0503 -89.9896>}]
 Wing_Lower_r1 [{shapeType=Capsule, length=38.7779, radius=27.6719, center=<-18.2959 8.5961 0.0187>, rotation=<rot 0 171.5041 -89.9336>}]
 Wing_Lower_r2 [{shapeType=Capsule, length=24.0851, radius=14.0993, center=<-19.3188 4.3646 0.018>, rotation=<rot 0 -7.3823 -89.9964>}]
 Wing_Lower_r3 [{shapeType=Capsule, length=38.4248, radius=6.0773, center=<-36.9342 -2.4413 0.0179>, rotation=<rot 0 -1.4304 -89.9377>}]
 Wing_Upper_l [{shapeType=Capsule, length=47.9526, radius=17.3904, center=<42.561 6.3533 0.0614>, rotation=<rot 0 -64.2923 -89.991>}]
 Wing_Upper_l1 [{shapeType=Capsule, length=68.7264, radius=52.107, center=<50.8483 3.9677 0.0602>, rotation=<rot 0 -67.8281 -89.9838>}]
 Wing_Upper_l2 [{shapeType=Capsule, length=31.4754, radius=21.7405, center=<25.0245 -8.0848 0.0652>, rotation=<rot 0 -33.3095 -89.9545>}]
 Wing_Upper_l3 [{shapeType=Capsule, length=36.8392, radius=18.0842, center=<-32.3357 -11.8858 0.0633>, rotation=<rot 0 74.3917 -89.8893>}]
 Wing_Upper_r [{shapeType=Capsule, length=47.9526, radius=17.3904, center=<-42.5611 -6.3533 -0.0614>, rotation=<rot 0 -64.2923 -89.991>}]
 Wing_Upper_r1 [{shapeType=Capsule, length=68.7264, radius=52.107, center=<-50.8483 -3.968 -0.0602>, rotation=<rot 0 -67.8279 -89.9838>}]
 Wing_Upper_r2 [{shapeType=Capsule, length=31.4754, radius=21.7405, center=<-25.0243 8.0845 -0.0652>, rotation=<rot 0 -33.3092 -89.9545>}]
 Wing_Upper_r3 [{shapeType=Capsule, length=36.8392, radius=18.0842, center=<32.3356 11.8859 -0.0633>, rotation=<rot 0 74.392 -89.8893>}]
}
constraints [
 {bone1Name=Wing_Upper_l, bone2Name=Frame, swing1LimitDegrees=45, swing2LimitDegrees=45, twistLimitDegrees=45}
 {bone1Name=Wing_Upper_l1, bone2Name=Wing_Upper_l, swing1LimitDegrees=45, swing2LimitDegrees=45, twistLimitDegrees=45}
 {bone1Name=Wing_Upper_l2, bone2Name=Wing_Upper_l1, swing1LimitDegrees=45, swing2LimitDegrees=45, twistLimitDegrees=45}
 {bone1Name=Wing_Upper_l3, bone2Name=Wing_Upper_l1, swing1LimitDegrees=45, swing2LimitDegrees=45, twistLimitDegrees=45}
 {bone1Name=Wing_Lower_l, bone2Name=Frame, swing1LimitDegrees=45, swing2LimitDegrees=45, twistLimitDegrees=45}
 {bone1Name=Wing_Lower_l1, bone2Name=Wing_Lower_l, swing1LimitDegrees=45, swing2LimitDegrees=45, twistLimitDegrees=45}
 {bone1Name=Wing_Lower_l2, bone2Name=Wing_Lower_l1, swing1LimitDegrees=45, swing2LimitDegrees=45, twistLimitDegrees=45}
 {bone1Name=Wing_Lower_l3, bone2Name=Wing_Lower_l1, swing1LimitDegrees=45, swing2LimitDegrees=45, twistLimitDegrees=45}
 {bone1Name=Wing_Upper_r, bone2Name=Frame, swing1LimitDegrees=45, swing2LimitDegrees=45, twistLimitDegrees=45}
 {bone1Name=Wing_Upper_r1, bone2Name=Wing_Upper_r, swing1LimitDegrees=45, swing2LimitDegrees=45, twistLimitDegrees=45}
 {bone1Name=Wing_Upper_r2, bone2Name=Wing_Upper_r1, swing1LimitDegrees=45, swing2LimitDegrees=45, twistLimitDegrees=45}
 {bone1Name=Wing_Upper_r3, bone2Name=Wing_Upper_r1, swing1LimitDegrees=45, swing2LimitDegrees=45, twistLimitDegrees=45}
 {bone1Name=Wing_Lower_r, bone2Name=Frame, swing1LimitDegrees=45, swing2LimitDegrees=45, twistLimitDegrees=45}
 {bone1Name=Wing_Lower_r1, bone2Name=Wing_Lower_r, swing1LimitDegrees=45, swing2LimitDegrees=45, twistLimitDegrees=45}
 {bone1Name=Wing_Lower_r3, bone2Name=Wing_Lower_r1, swing1LimitDegrees=45, swing2LimitDegrees=45, twistLimitDegrees=45}
 {bone1Name=Wing_Lower_r2, bone2Name=Wing_Lower_r1, swing1LimitDegrees=45, swing2LimitDegrees=45, twistLimitDegrees=45}
 {bone1Name=Handles_Base, bone2Name=Frame, swing1LimitDegrees=45, swing2LimitDegrees=45, twistLimitDegrees=45}
 {bone1Name=Handle_r, bone2Name=Handles_Base, swing1LimitDegrees=45, swing2LimitDegrees=45, twistLimitDegrees=45}
 {bone1Name=Handle_l, bone2Name=Handles_Base, swing1LimitDegrees=45, swing2LimitDegrees=45, twistLimitDegrees=45}
]
